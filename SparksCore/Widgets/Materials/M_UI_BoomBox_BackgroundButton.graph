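# Translucent UI material that draws the procedural background and button shapes for a BoomBox button widget, tinted by a Color parameter.
# SDF circle and box functions form a trimmed ring: a box-cut outer circle minus an inner circle-box union, with smooth rounded corners.
# The button variant subtracts a side box with adjustable padding and rounding, then applies an Opacity parameter.
# Both shapes are anti-aliased, and the IsBackgrond static switch selects between them.
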Material /SparksCore/Widgets/Materials/M_UI_BoomBox_BackgroundButton
graph {
 properties {bAutomaticallySetUsageInEditor=false, blendMode=BLEND_Translucent, materialDomain=MD_UI, usedShadingModels=Default Lit}
 nodes [
  n1 = Constant() {r=1}
  n2 = Add(A=AppendVector(A=Constant(), B=ScalarParameter() {defaultValue=-0.5, parameterName=TrimBox}), B=TextureCoordinate())
  n3 = Constant() {r=1}
  n4 = ScalarParameter() {defaultValue=0.5, parameterName=InnerCircleSize}
  n5 = Add(A=AppendVector(A=Constant(), B=ScalarParameter() {defaultValue=-0.5, parameterName=TrimBox}), B=TextureCoordinate())
  n6 = Constant() {r=1}
  n7 = MaterialFunctionCall!(
   DistanceFieldA=MaterialFunctionCall(
    DistanceFieldA=MaterialFunctionCall(Size=@n1) {materialFunction=&MF_UI_Circle}.SDF
    DistanceFieldB=MaterialFunctionCall(UVs=@n2, Size=@n3) {materialFunction=&MF_UI_Box}.SDF
    BlendAmount=ScalarParameter() {defaultValue=0.172, parameterName=ExternalRoundedCorners}
   ) {materialFunction=/FortUILibrary/MaterialFunctions/Shapes/Modifiers/MF_UI_SDF_Subtraction}.Smooth
   DistanceFieldB=MaterialFunctionCall!(
    DistanceFieldA=MaterialFunctionCall(Size=@n4) {materialFunction=&MF_UI_Circle}.SDF
    DistanceFieldB=MaterialFunctionCall(UVs=@n5, Size=@n6) {materialFunction=&MF_UI_Box}.SDF
    BlendAmount=ScalarParameter() {defaultValue=0.1054, parameterName=InternalRoundedCorners}
   ) {materialFunction=/FortUILibrary/MaterialFunctions/Shapes/Modifiers/MF_UI_SDF_Union}.Smooth
   BlendAmount
  ) {materialFunction=&MF_UI_SDF_Subtraction}
  NamedRerouteDeclaration(None=MaterialFunctionCall(DistanceField=NamedRerouteDeclaration(None=@n7.Exact) {name=SDFBackground, nodeColor=<rgba 0 1 0.5547 1>}) {materialFunction=&MF_UI_AntiAliasing}.Inner) {name=BackgroundFill, nodeColor=<rgba 1 0 0.9609 1>}
  n8 = Add(A=AppendVector(A=Constant() {r=0.5}, B=Constant()), B=TextureCoordinate())
  n9 = AppendVector(A=Add(A=Constant() {r=1}, B=ScalarParameter() {defaultValue=0.2, parameterName=InternalPadding}), B=Constant() {r=1})
  n10 = MaterialFunctionCall(DistanceField=MaterialFunctionCall(
    DistanceFieldA=NamedRerouteUsage()
    DistanceFieldB=MaterialFunctionCall(UVs=@n8, Size=@n9) {materialFunction=&MF_UI_Box}.SDF
    BlendAmount=ScalarParameter() {defaultValue=0.1054, parameterName=AxialVerticalRoundedCorners}
   ) {materialFunction=&MF_UI_SDF_Subtraction}.Smooth) {materialFunction=&MF_UI_AntiAliasing}
  NamedRerouteDeclaration(None=LinearInterpolate(B=Multiply(A=@n10.Inner), Alpha=ScalarParameter() {defaultValue=1, parameterName=Opacity})) {name=ButtonFill, nodeColor=<rgba 1 0.2344 0 1>}
  NamedRerouteDeclaration(None=VectorParameter() {parameterName=Color, defaultValue=<rgba 1 0 0.9851 1>}.RGB) {name=Color, nodeColor=<rgba 0 0.0859 1 1>}
  NamedRerouteUsage()
  StaticSwitchParameter(True=NamedRerouteUsage(), False=NamedRerouteUsage()) {defaultValue=true, parameterName=IsBackgrond}
 ]
}
