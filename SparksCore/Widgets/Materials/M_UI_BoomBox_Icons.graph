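# Translucent UI material for BoomBox widget icons, with a speaker texture as the default icon.
# The icon's MSDF texture converts to an SDF and passes through a thresholded anti-aliasing function to form a crisp shape mask.
# A static switch can multiply that mask by a smoothstepped vertical gradient that fades its lower part.
# Icon color blends between ColorSelected and ColorNormal using the b_Selected scalar.
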Material /SparksCore/Widgets/Materials/M_UI_BoomBox_Icons
graph {
 properties {bAutomaticallySetUsageInEditor=false, blendMode=BLEND_Translucent, materialDomain=MD_UI, usedShadingModels=Default Lit}
 nodes [
  NamedRerouteDeclaration(None=LinearInterpolate(
    A=VectorParameter() {parameterName=ColorSelected, defaultValue=<rgba 0.0048 0.004 0.0116 1>}.RGB
    B=VectorParameter() {parameterName=ColorNormal, defaultValue=<rgba 1 1 1 1>}.RGB
    Alpha=ScalarParameter() {parameterName=b_Selected}
   )) {name=Color, nodeColor=<rgba 0 1 0.8594 1>}
  n1 = LinearInterpolate(Alpha=SmoothStep(
    Min=ScalarParameter() {defaultValue=0.75, parameterName=TopPosition}
    Max=ScalarParameter() {defaultValue=1, parameterName=ButtonPosition}
    Value=MaterialFunctionCall() {materialFunction=&LinearGradient}.VGradient
   )) {constA=1, constB=-0.2}
  n2 = MaterialFunctionCall(
   DistanceField=MaterialFunctionCall!(MSDF RGB=TextureSampleParameter2D() {parameterName=IconTexture, texture=/SparksCore/Widgets/Textures/T_UI_Boombox_Speaker}.RGB) {materialFunction=/FortUILibrary/MaterialFunctions/Shapes/Modifiers/MF_UI_MSDFtoSDF}.SDF
   Threshold=ScalarParameter() {defaultValue=0.15, parameterName=Threshold}
  ) {materialFunction=&MF_UI_AntiAliasing}
  NamedRerouteDeclaration(None=StaticSwitchParameter(True=Multiply(A=@n1, B=@n2.Outer), False=@n2.Outer) {parameterName=HasGradientMask}) {name=Texture, nodeColor=<rgba 1 0.5625 0 1>}
  NamedRerouteUsage()
  NamedRerouteUsage()
 ]
}
