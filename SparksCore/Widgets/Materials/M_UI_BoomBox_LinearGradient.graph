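# Translucent UI material for the BoomBox widget that draws a vertical two-color gradient with exposed color parameters, TopColor and ButtonColor.
# A SmoothStep between TopPosition and ButtonPosition remaps LinearGradient's VGradient to blend the colors.
# A separate blend over the raw VGradient, apparently of both colors' alphas via named reroutes, multiplies MF_UI_Box's Fill as a shape mask, so the position parameters affect only the color blend.
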Material /SparksCore/Widgets/Materials/M_UI_BoomBox_LinearGradient
graph {
 properties {bAutomaticallySetUsageInEditor=false, blendMode=BLEND_Translucent, materialDomain=MD_UI, usedShadingModels=Default Lit}
 nodes [
  n1 = VectorParameter() {parameterName=TopColor, defaultValue=<rgba 1 0 0.6547 1>}
  n2 = VectorParameter() {parameterName=ButtonColor, defaultValue=<rgba 0 0.0004 1 1>}
  LinearInterpolate(
   A=@n1.RGBA
   B=@n2.RGBA
   Alpha=SmoothStep(
    Min=ScalarParameter() {parameterName=TopPosition}
    Max=ScalarParameter() {defaultValue=1, parameterName=ButtonPosition}
    Value=MaterialFunctionCall() {materialFunction=&LinearGradient}.VGradient
   )
  )
  Multiply(
   A=MaterialFunctionCall(Size=Constant() {r=1}, CornerRoundness=Constant() {r=1}) {materialFunction=&MF_UI_Box}.Fill
   B=LinearInterpolate(
    A=NamedRerouteUsage()
    B=NamedRerouteUsage()
    Alpha=MaterialFunctionCall() {materialFunction=&LinearGradient}.VGradient
   )
  )
  NamedRerouteDeclaration(None=@n2.A) {name=OpacityButton, nodeColor=<rgba 0.4531 1 0 1>}
  NamedRerouteDeclaration(None=@n1.A) {name=OpacityTop, nodeColor=<rgba 0.4531 1 0 1>}
 ]
}
